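# Physics asset for the Skeleton Hunter backpack cosmetic, providing simple collision and joint limits for its dangling cloth bones.
# A capsule body covers the pack, and box bodies form left, middle, mid-left and right cloth chains constrained back to it.
# The right chain branches at its third bone into a fourth segment and two side pieces.
# Twist is tightly limited throughout, with wide swing on side chains and narrower swing on middle ones.
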PhysicsAsset /BRCosmetics/Accessories/FORT_Backpacks/M_MED_SkeletonHunter_Backpack/Meshes/M_MED_Skeleton_Hunter_Pack_Physics
bodies {
 dyn_l_cloth_01 [{shapeType=Box, extentX=7, extentY=6, extentZ=3, center=<2 0 0>}]
 dyn_l_cloth_02 [{shapeType=Box, extentX=7, extentY=6, extentZ=3, center=<2 0 0>}]
 dyn_l_cloth_03 [{shapeType=Box, extentX=7, extentY=6, extentZ=3, center=<2 0 0>}]
 dyn_mid_cloth_01 [{shapeType=Box, extentX=6, extentY=10, extentZ=1, center=<2 0 0>}]
 dyn_mid_cloth_02 [{shapeType=Box, extentX=6, extentY=10, extentZ=1, center=<2 0 0>}]
 dyn_mid_cloth_03 [{shapeType=Box, extentX=6, extentY=10, extentZ=1, center=<0.9532 0 0>}]
 dyn_mid_l_cloth_01 [{shapeType=Box, extentX=6, extentY=10, extentZ=1, center=<2 0 0>}]
 dyn_mid_l_cloth_02 [{shapeType=Box, extentX=6, extentY=10, extentZ=1, center=<2 0 0>}]
 dyn_r_cloth_01 [{shapeType=Box, extentX=8.5797, extentY=7.5, extentZ=3, center=<2 0 0>, rotation=<rot -0.0005 0 -8.707>}]
 dyn_r_cloth_02 [{shapeType=Box, extentX=7.4576, extentY=7.5, extentZ=3, center=<2 0 0>, rotation=<rot -0.0005 0 -8.707>}]
 dyn_r_cloth_03 [{shapeType=Box, extentX=8.7197, extentY=8.6315, extentZ=3, center=<2 0 0>, rotation=<rot -0.0005 0 7.5>}]
 dyn_r_cloth_03_a [{shapeType=Box, extentX=13.5634, extentY=3, extentZ=3.8595, center=<5.2361 1.3094 0.0115>}]
 dyn_r_cloth_03_b [{shapeType=Box, extentX=9.8431, extentY=1.2257, extentZ=2.4174, center=<4.1425 0.0671 0.146>, rotation=<rot -9.776 -4.5433 25.0814>}]
 dyn_r_cloth_04 [{shapeType=Box, extentX=17.294, extentY=7.5, extentZ=3, center=<6.7495 0 0>, rotation=<rot -0.0005 0 -8.7069>}]
 pack [{shapeType=Capsule, length=13.6499, radius=16.3501, center=<-0.1853 14.6476 -16.7441>, rotation=<rot -0.4351 4.9812 11.8793>}]
}
constraints [
 {bone1Name=dyn_mid_cloth_01, bone2Name=pack, swing1LimitDegrees=30, swing2LimitDegrees=60, twistLimitDegrees=2}
 {bone1Name=dyn_mid_cloth_02, bone2Name=dyn_mid_cloth_01, swing1LimitDegrees=30, swing2LimitDegrees=60, twistLimitDegrees=2}
 {bone1Name=dyn_mid_cloth_03, bone2Name=dyn_mid_cloth_02, swing1LimitDegrees=30, swing2LimitDegrees=60, twistLimitDegrees=2}
 {bone1Name=dyn_r_cloth_01, bone2Name=pack, swing1LimitDegrees=90, swing2LimitDegrees=50, twistLimitDegrees=2}
 {bone1Name=dyn_r_cloth_02, bone2Name=dyn_r_cloth_01, swing1LimitDegrees=90, swing2LimitDegrees=50, twistLimitDegrees=2}
 {bone1Name=dyn_r_cloth_03, bone2Name=dyn_r_cloth_02, swing1LimitDegrees=90, swing2LimitDegrees=55, twistLimitDegrees=2}
 {bone1Name=dyn_r_cloth_04, bone2Name=dyn_r_cloth_03, swing1LimitDegrees=90, swing2LimitDegrees=60, twistLimitDegrees=2}
 {bone1Name=dyn_r_cloth_03_a, bone2Name=dyn_r_cloth_03, swing1LimitDegrees=90, swing2LimitDegrees=60, twistLimitDegrees=2}
 {bone1Name=dyn_l_cloth_01, bone2Name=pack, swing1LimitDegrees=90, swing2LimitDegrees=60, twistLimitDegrees=2}
 {bone1Name=dyn_l_cloth_02, bone2Name=dyn_l_cloth_01, swing1LimitDegrees=90, swing2LimitDegrees=60, twistLimitDegrees=2}
 {bone1Name=dyn_l_cloth_03, bone2Name=dyn_l_cloth_02, swing1LimitDegrees=90, swing2LimitDegrees=60, twistLimitDegrees=2}
 {bone1Name=dyn_mid_l_cloth_01, bone2Name=pack, swing1LimitDegrees=30, swing2LimitDegrees=10, twistLimitDegrees=2}
 {bone1Name=dyn_mid_l_cloth_02, bone2Name=dyn_mid_l_cloth_01, swing1LimitDegrees=30, swing2LimitDegrees=10, twistLimitDegrees=2}
 {bone1Name=dyn_r_cloth_03_b, bone2Name=dyn_r_cloth_03, swing1LimitDegrees=90, swing2LimitDegrees=60, twistLimitDegrees=2}
]
