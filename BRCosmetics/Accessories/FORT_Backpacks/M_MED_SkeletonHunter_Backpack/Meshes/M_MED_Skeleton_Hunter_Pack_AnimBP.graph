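# Animation Blueprint for the Skeleton Hunter backpack cosmetic, adding pose-driven tilt and physics motion to the pack.
# Each update maps spine_05 yaw relative to spine_01 into a clamped roll, applied additively to the pack bone.
# A RigidBody node then simulates the pack's physics asset, using the skydiving vector force as external force.
# World gravity is overridden, weakening during skydiving until the parachute opens.
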AnimBlueprint /BRCosmetics/Accessories/FORT_Backpacks/M_MED_SkeletonHunter_Backpack/Meshes/M_MED_Skeleton_Hunter_Pack_AnimBP
target_skeleton=&/M_MED_Skeleton_Hunter_Pack_Skeleton
graphs [
 AnimGraph {nodes [
   n1 = "Transform (Modify) Bone - Bone: pack"(ComponentPose=Local To Component(LocalPose=Input Pose()), Translation="0,0,0", Rotation=Get backpack_rotate()) {
    rotationMode=BMM_Additive
    rotationSpace=BCS_BoneSpace
    translationMode=BMM_Additive
    translationSpace=BCS_BoneSpace
    alphaBoolBlend {blendOption=Linear}
    boneToModify {boneName=pack}
   }
   n2 = RigidBody!(
    ComponentPose=@n1
    Alpha
    OverrideWorldGravity:Vector=Get gravity_override() -> (gravity_override:Vector)
    ExternalForce=Break Fort Anim Input Skydiving External Force(FortAnimInput_SkydivingExternalForce=Get SkydivingExternalForceInput()) -> (
     bUseSkydivingVectorForce=~
     SkydivingVectorForce=~
     HeadToPelvisDirection=~
     FloatingMultiplier=~
     FloatingAdditive=~
     DivingMultiplier=~
     DivingAdditive=~
     ParachutingMultiplier=~
     ParachutingAdditive=~
    ).SkydivingVectorForce
   ) -> (Pose) {
    bClampLinearTranslationLimitToRefPose=true
    bOverrideWorldGravity=true
    overridePhysicsAsset=&/M_MED_Skeleton_Hunter_Pack_Physics
    alphaBoolBlend {blendOption=Linear}
    componentAppliedLinearAccClamp=<3000 3000 3000>
    componentLinearAccScale=<0.7 0.5 0.5>
    componentLinearVelScale=<0.7 0.5 0.5>
    overrideWorldGravity {z=-980}
   }
   Output Pose(Result=Component To Local(ComponentPose=@n2)) {layerGroup=None}
  ]}
 EventGraph {nodes [
   Event BlueprintUpdateAnimation()
   n1 = GetOwnerSkeletalMeshComponent()
   n2 = BreakRotator(InRot=BreakTransform(InTransform=GetDeltaTransformFromRefPose(self=@n1, BoneName=spine_05, BaseName=spine_01)).Rotation)
   > Set backpack_rotate(backpack_rotate=MakeRotator(Roll=MapRangeClamped(Value=@n2.Yaw, InRangeA=0, InRangeB=-30, OutRangeB=30)))
   > n3 = Branch(Condition=Get bIsSkyDiving())
   > n4 = Branch(Condition=Get bIsParachuteOpened())
   Set gravity_override(execute=[@n4.then, @n3.else], gravity_override="0, 0,-980")
   Set gravity_override(execute=@n4.else, gravity_override="0, 0,-300")
   TryGetPawnOwner()
   Branch(Condition=Get bIsSkyDiving())
  ]}
]
variables [
 backpack_rotate {type=Rotator}
 gravity_override {type=Vector}
]
